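# Sound cue for the wing-flap foley of a biped chicken cosmetic companion.
# A quality-level switch feeds either an eight-variation random pool in the upper two slots or a four-variation subset of the same waves in the lower two.
# A modulator at the root adds slight pitch and volume variation.
# It plays as a pet creature sound with companion foley attenuation, borrowing the quadruped footstep concurrency setting.
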SoundCue /CosmeticCompanions/Assets/Audio_Shared/Foley/Biped/Chicken/SC_Mimosa_Biped_Chicken_Flap
root=Modulator_1
nodes [
 {id=Modulator_1, children [QualityLevel_1], properties {pitchMax=1.1, pitchMin=0.9, volumeMax=1.1, volumeMin=0.9}}
 {id=QualityLevel_1, children [Random_0,Random_0,Random_1,Random_1]}
 {
  id=Random_0
  children [WavePlayer_3, WavePlayer_4, WavePlayer_5, WavePlayer_6, WavePlayer_7, WavePlayer_8, WavePlayer_9, WavePlayer_10]
  properties {weights [1,1,1,1,1,1,1,1]}
 }
 {id=Random_1, children [WavePlayer_17, WavePlayer_18, WavePlayer_19, WavePlayer_27], properties {weights [1,1,1,1]}}
 {
  id=WavePlayer_10
  properties {soundWaveAssetPtr=/CosmeticCompanions/Assets/Audio_Shared/Foley/Biped/Chicken/Wavs/Mimosa_Biped_Chicken_Flap_01}
 }
 {
  id=WavePlayer_17
  properties {soundWaveAssetPtr=/CosmeticCompanions/Assets/Audio_Shared/Foley/Biped/Chicken/Wavs/Mimosa_Biped_Chicken_Flap_04}
 }
 {
  id=WavePlayer_18
  properties {soundWaveAssetPtr=/CosmeticCompanions/Assets/Audio_Shared/Foley/Biped/Chicken/Wavs/Mimosa_Biped_Chicken_Flap_03}
 }
 {
  id=WavePlayer_19
  properties {soundWaveAssetPtr=/CosmeticCompanions/Assets/Audio_Shared/Foley/Biped/Chicken/Wavs/Mimosa_Biped_Chicken_Flap_02}
 }
 {
  id=WavePlayer_27
  properties {soundWaveAssetPtr=&Mimosa_Biped_Chicken_Flap_01}
 }
 {
  id=WavePlayer_3
  properties {soundWaveAssetPtr=/CosmeticCompanions/Assets/Audio_Shared/Foley/Biped/Chicken/Wavs/Mimosa_Biped_Chicken_Flap_08}
 }
 {
  id=WavePlayer_4
  properties {soundWaveAssetPtr=/CosmeticCompanions/Assets/Audio_Shared/Foley/Biped/Chicken/Wavs/Mimosa_Biped_Chicken_Flap_07}
 }
 {
  id=WavePlayer_5
  properties {soundWaveAssetPtr=/CosmeticCompanions/Assets/Audio_Shared/Foley/Biped/Chicken/Wavs/Mimosa_Biped_Chicken_Flap_06}
 }
 {
  id=WavePlayer_6
  properties {soundWaveAssetPtr=/CosmeticCompanions/Assets/Audio_Shared/Foley/Biped/Chicken/Wavs/Mimosa_Biped_Chicken_Flap_05}
 }
 {
  id=WavePlayer_7
  properties {soundWaveAssetPtr=&Mimosa_Biped_Chicken_Flap_04}
 }
 {
  id=WavePlayer_8
  properties {soundWaveAssetPtr=&Mimosa_Biped_Chicken_Flap_03}
 }
 {
  id=WavePlayer_9
  properties {soundWaveAssetPtr=&Mimosa_Biped_Chicken_Flap_02}
 }
]
properties {
 attenuationSettings=/Game/Sounds/Attenuation/AS_Companion_Foley
 maxDistance=1200
 priority=0
 soundClassObject=&Fort_Creatures_Pets
 volumeMultiplier=1.6
 concurrencySet [/CosmeticCompanions/Assets/Audio_Shared/Foley/Quadraped/CONC_Mimosa_Quadraped_Footstep]
}
