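# Sound cue for the scratch foley of a chicken biped cosmetic companion, routed to the pets sound class with companion foley attenuation.
# A quality level node picks between two equally weighted random pools: five scratch waves for the first two levels, four for the last two.
# A root modulator adds slight volume variation, and playback shares the biped footstep concurrency set.
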SoundCue /CosmeticCompanions/Assets/Audio_Shared/Foley/Biped/Chicken/SC_Mimosa_Biped_Chicken_Scratch
root=Modulator_0
nodes [
 {id=Modulator_0, children [QualityLevel_0], properties {volumeMax=1.1, volumeMin=0.9}}
 {id=QualityLevel_0, children [Random_3,Random_3,Random_5,Random_5]}
 {id=Random_3, children [WavePlayer_11, WavePlayer_12, WavePlayer_13, WavePlayer_14, WavePlayer_15], properties {weights [1,1,1,1,1]}}
 {id=Random_5, children [WavePlayer_16, WavePlayer_19, WavePlayer_20, WavePlayer_21], properties {weights [1,1,1,1]}}
 {
  id=WavePlayer_11
  properties {soundWaveAssetPtr=/CosmeticCompanions/Assets/Audio_Shared/Foley/Biped/Chicken/Wavs/Mimosa_Biped_Chicken_Scratch_05}
 }
 {
  id=WavePlayer_12
  properties {soundWaveAssetPtr=/CosmeticCompanions/Assets/Audio_Shared/Foley/Biped/Chicken/Wavs/Mimosa_Biped_Chicken_Scratch_04}
 }
 {
  id=WavePlayer_13
  properties {soundWaveAssetPtr=/CosmeticCompanions/Assets/Audio_Shared/Foley/Biped/Chicken/Wavs/Mimosa_Biped_Chicken_Scratch_03}
 }
 {
  id=WavePlayer_14
  properties {soundWaveAssetPtr=/CosmeticCompanions/Assets/Audio_Shared/Foley/Biped/Chicken/Wavs/Mimosa_Biped_Chicken_Scratch_02}
 }
 {
  id=WavePlayer_15
  properties {soundWaveAssetPtr=/CosmeticCompanions/Assets/Audio_Shared/Foley/Biped/Chicken/Wavs/Mimosa_Biped_Chicken_Scratch_01}
 }
 {
  id=WavePlayer_16
  properties {soundWaveAssetPtr=&Mimosa_Biped_Chicken_Scratch_04}
 }
 {
  id=WavePlayer_19
  properties {soundWaveAssetPtr=&Mimosa_Biped_Chicken_Scratch_03}
 }
 {
  id=WavePlayer_20
  properties {soundWaveAssetPtr=&Mimosa_Biped_Chicken_Scratch_02}
 }
 {
  id=WavePlayer_21
  properties {soundWaveAssetPtr=&Mimosa_Biped_Chicken_Scratch_01}
 }
]
properties {
 attenuationSettings=/Game/Sounds/Attenuation/AS_Companion_Foley
 maxDistance=1200
 priority=0
 soundClassObject=&Fort_Creatures_Pets
 volumeMultiplier=0.5
 concurrencySet [/CosmeticCompanions/Assets/Audio_Shared/Foley/Biped/CONC_Mimosa_Biped_Footstep]
}
